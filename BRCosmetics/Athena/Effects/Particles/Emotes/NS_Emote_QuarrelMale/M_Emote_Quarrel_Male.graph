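# Translucent, unlit, two-sided Niagara mesh-particle material for the male Quarrel emote effect.
# Panned streak, distorted beam and cartoon-flame textures combine with a threshold dissolve, driven by dynamic parameters.
# A Light scalar blends this to a scalable candle-texture light; Glow blends to a soft sphere-mask glow.
# Color parameters tint emissive by masks, particle alpha scales opacity, and depth fade softens the non-glow branch.
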Material /BRCosmetics/Athena/Effects/Particles/Emotes/NS_Emote_QuarrelMale/M_Emote_Quarrel_Male
graph {
 properties {
  bUsedWithNiagaraMeshParticles=true
  blendMode=BLEND_Translucent
  pixelDepthOffsetMode=PDOM_Legacy
  shadingModel=MSM_Unlit
  twoSided=true
  usedShadingModels=Unlit
 }
 nodes [
  n1 = TextureCoordinate()
  n2 = DynamicParameter() {defaultValue=<rgba -1 1 1 1>, paramNames [UV, Light, Dissolveandlightsize, Noise]}
  n3 = TextureSample(UVs=Panner(Coordinate=MaterialFunctionCall(XY=@n1) {materialFunction=&Swizzle}.YX, Time=@n2.UV) {speedY=1}) {samplerSource=SSM_Clamp_WorldGroupSettings, texture=/Game/Effects/Fort_Effects/Gameplay/SmokeDustExplosions/Textures/T_Explosion_Streak_02}
  n4 = Multiply(A=@n3.R) {constB=8}
  n5 = MaterialFunctionCall(XY=TextureCoordinate()) {materialFunction=&Swizzle}
  n6 = AppendVector(A=Constant(), B=Constant() {r=-1})
  n7 = TextureSample(UVs=Panner(Coordinate=TextureCoordinate() {uTiling=0.5, vTiling=0.3}) {speedX=0.3}) {samplerType=SAMPLERTYPE_Normal, texture=&T_Weap_Distort_N_02}
  n8 = TextureSample(UVs=Add(A=Panner(Coordinate=@n5.YX, Speed=@n6), B=Multiply(A=@n7.R) {constB=0.14})) {texture=/Game/Global/Textures/Default/Masks_and_Gradients/T_BlasterBeamDistort}
  n9 = Multiply(A=TextureSample(UVs=Panner(Time=@n2.UV) {speedX=1}) {samplerSource=SSM_Clamp_WorldGroupSettings, texture=/Game/Effects/Fort_Effects/PlayerHealthDamage/T_HitStreak}.R) {constB=1.4}
  n10 = TextureCoordinate() {uTiling=0.42, vTiling=0.85}
  n11 = TextureSample(UVs=MaterialFunctionCall(XY=Panner(Coordinate=@n10) {speedX=1}) {materialFunction=&Swizzle}.YX) {texture=/Game/Effects/Fort_Effects/Textures/Tile/T_Tile_Texture_Cartoon_Flames_Layer}
  n12 = SphereMask(A=ComponentMask(None=TextureCoordinate()) {g=true}, B=Constant() {r=0.5}) {attenuationRadius=0.31, hardnessPercent=0}
  n13 = Multiply(A=@n12, B=@n12)
  n14 = TextureSample(UVs=MaterialFunctionCall(XY=TextureCoordinate() {uTiling=1.5, vTiling=3}) {materialFunction=&Swizzle}.YX) {texture=&T_BlasterBeamDistort}
  n15 = Add(A=@n13, B=@n14.R)
  n16 = DynamicParameter() {defaultValue=<rgba -1 1 1 1>, paramNames [UV, Light, Dissolveandlightsize, Noise]}
  n17 = Constant() {r=1}
  n18 = Multiply(
   A=Add(A=Multiply(A=@n4, B=@n8.R), B=Multiply(A=@n9, B=@n11.G))
   B=If(A=Clamp(None=@n15), B=@n16.Dissolveandlightsize, "A > B"=@n17, "A == B"=@n17, "A < B"=Constant())
  )
  n19 = Multiply(A=Multiply(A=@n14.R, B=@n14.R), B=@n16.Noise) {constB=10}
  n20 = ScalarParameter() {defaultValue=1}
  n21 = DynamicParameter() {defaultValue=<rgba -1 1 1 1>, paramNames [UV, Light, Dissolveandlightsize, Noise]}
  n22 = TextureSample(UVs=MaterialFunctionCall(Texture Scale=AppendVector(A=@n20, B=@n21.Dissolveandlightsize)) {materialFunction=&ScaleUVsByCenter}.UVs) {
   samplerSource=SSM_Clamp_WorldGroupSettings
   samplerType=SAMPLERTYPE_LinearColor
   texture=/BRCosmetics/Effects/Fort_Effects/Textures/Fire/Candle
  }
  n23 = Multiply(
   A=Multiply(A=@n22.R) {constB=2}
   B=DynamicParameter() {defaultValue=<rgba -1 1 1 1>, paramNames [UV, Light, Dissolveandlightsize, Noise]}.Light
  )
  n24 = TextureCoordinate()
  n25 = ComponentMask(None=@n24) {g=true}
  n26 = Constant() {r=0.5}
  n27 = SphereMask(A=@n25, B=@n26) {attenuationRadius=0.5, hardnessPercent=50}
  n28 = ComponentMask(None=@n24) {r=true}
  n29 = SphereMask(A=@n28, B=@n26) {attenuationRadius=0.5, hardnessPercent=0}
  n30 = Multiply(A=@n25) {constB=0.5}
  n31 = Multiply(A=OneMinus(None=Sine(None=@n30)), B=@n28)
  n32 = TextureSample(UVs=TextureCoordinate() {uTiling=1.3, vTiling=2}) {samplerType=SAMPLERTYPE_Grayscale, texture=&LowResBlurredNoise}
  n33 = Multiply(A=@n32.R, B=@n32.R)
  n34 = Subtract(A=@n29, B=Multiply(A=Multiply(A=@n31, B=@n27), B=Multiply(A=@n33) {constB=8}))
  n35 = Multiply(A=@n29, B=@n29)
  n36 = Multiply(
   A=LinearInterpolate(A=Add(A=@n18, B=@n19), B=@n23, Alpha=ScalarParameter() {defaultValue=1, parameterName=Light})
   B=LinearInterpolate(A=Multiply(A=@n27, B=@n34), B=Multiply(A=@n27, B=@n35), Alpha=ScalarParameter() {defaultValue=1, parameterName=Light})
  )
  n37 = ParticleColor()
  n38 = TextureCoordinate()
  n39 = Constant() {r=0.5}
  n40 = SphereMask(
   A=ComponentMask(None=@n38) {g=true}
   B=@n39
   Radius=Multiply(
    A=Constant() {r=0.5}
    B=DynamicParameter() {defaultValue=<rgba -1 1 1 1>, paramNames [UV, Light, Dissolveandlightsize, Noise]}.Dissolveandlightsize
   )
  ) {attenuationRadius=0.5, hardnessPercent=80}
  n41 = SphereMask(A=ComponentMask(None=@n38) {r=true}, B=@n39) {attenuationRadius=0.5, hardnessPercent=0}
  LinearInterpolate(
   A=DepthFade(Opacity=Multiply(A=@n36, B=@n37.A))
   B=Multiply(A=@n37.A, B=Multiply(A=@n40, B=@n41))
   Alpha=ScalarParameter() {parameterName=Glow}
  )
  n42 = LinearInterpolate(
   A=VectorParameter() {parameterName=DarkColor, defaultValue=<rgba 0.2292 0.2196 0.0561 1>}.RGB
   B=LinearInterpolate(
    A=VectorParameter() {parameterName=ColorB, defaultValue=<rgba 0.3073 0.3021 0.0158 1>}.RGB
    B=VectorParameter() {parameterName=ColorA, defaultValue=<rgba 5 4.1669 0.7031 1>}.RGB
    Alpha=Add(A=@n13, B=@n23)
   )
   Alpha=@n14.R
  )
  n43 = VectorParameter() {defaultValue=<rgba 0.3073 0.3021 0.0158 1>}
  n44 = VectorParameter() {parameterName=Param_1, defaultValue=<rgba 5 4.1669 0.7031 1>}
  n45 = SphereMask(A=ComponentMask(None=TextureCoordinate()) {g=true}, B=Constant() {r=0.5}) {attenuationRadius=0.31, hardnessPercent=0}
  n46 = Multiply(A=@n45, B=@n45)
  MaterialFunctionCall(Emissive=LinearInterpolate(A=Multiply(A=@n42, B=@n37.RGB), B=LinearInterpolate(A=@n43.RGB, B=@n44.RGB, Alpha=@n46), Alpha=ScalarParameter() {parameterName=Glow})) {materialFunction=&MF_FX_EmissiveAdjustments}
 ]
}
