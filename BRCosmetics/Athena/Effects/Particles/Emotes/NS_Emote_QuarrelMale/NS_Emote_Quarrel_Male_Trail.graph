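# Niagara trail effect for the male Quarrel emote: a brief spinning mesh flash with smoke, attached to the character's skeleton.
# Dark, Glow and Light each burst one short-lived mesh at the pelvis after a slight delay, with curve-driven spin, color and material parameters.
# Two Smoke emitters spray sprites from the root in cones, one angled higher, using drag, randomized rise or fall and time-of-day alpha scaling.
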NiagaraSystem /BRCosmetics/Athena/Effects/Particles/Emotes/NS_Emote_QuarrelMale/NS_Emote_Quarrel_Male_Trail
dependencies {
 usedDataInterfaces [&NiagaraDataInterfaceColorCurve, &NiagaraDataInterfaceCurve]
 usedDynamicInputs []
 usedModules [&SystemState]
}
emitters {
 Dark {
  emitterUpdateScript [
   &EmitterState(Life Cycle Mode=::NewEnumerator1, Loop Behavior=::NewEnumerator1, Loop Delay=0.06, Loop Duration=1, UseLoopDelay=true)
   &SpawnBurst_Instantaneous(Spawn Count=1, Use Spawn Probability=false)
  ]
  particleSpawnScript [
   &InitializeParticle(
    Color Mode=::NewEnumerator0
    Lifetime=0.28
    Mesh Scale=<5.5 5.5 5>
    Mesh Scale Mode=::NewEnumerator4
    UsePositionOffset=false
    Write Material Random=true
   )
   &SkeletalMeshLocation(
    Bone Sampling Mode=::NewEnumerator4
    MaskSpawnGroup=false
    Sampled Position Offset=<0 5 30>
    Sampled Position Offset Coordinate Space:ENiagaraSkelMeshTransforms=::NewEnumerator2
    Skeletal Mesh=DataInterface() {propertyValues {
      DefaultMesh=null
      ExcludeBoneName=root
      OptionalComponentClass=null
      PreviewMesh=&M_Med_Soldier_04_Preview
      SkinningMode=SkinOnTheFly
      SoftSourceActor=null
      SourceMode=Default
      UvSetIndex=0
      WholeMeshLOD=-1
      bExcludeBone=false
      bReadDeformedGeometry=true
      bRequireCurrentFrameData=true
      ComponentTags []
      FilteredBones [pelvis]
      FilteredSockets []
      MeshUserParameter {parameter {name=None, typeDefHandle {registeredTypeIndex=1}, varData []}}
      SamplingRegions []
     }}
   )
   &InitialMeshOrientation(
    Orientation Coordinate Space=::Local
    Rotation=<0.25 0 0>
    Rotation Coordinate Space=::NewEnumerator2
    Use Orientation Vector=true
    Use Rotation Vector=true
   )
  ]
  particleUpdateScript [
   &ParticleState()
   &ScaleColor(
    Scale Alpha=&FloatFromCurve(FloatCurve=DataInterface() {propertyValues {Curve {keys [
         {arriveTangent=2.5813, interpMode=RCIM_Cubic, leaveTangent=2.5813, tangentMode=RCTM_User, time=0.1, value=0}
         {arriveTangent=7.725, interpMode=RCIM_Cubic, leaveTangent=7.725, tangentMode=RCTM_User, time=0.3, value=0.625}
         {arriveTangent=2.6751, interpMode=RCIM_Cubic, leaveTangent=-1.5414, tangentMode=RCTM_Break, time=0.5, value=3}
         {arriveTangent=-1.8276, interpMode=RCIM_Cubic, leaveTangent=-1.8276, tangentMode=RCTM_User, time=1, value=0}
        ]}}})
    Scale RGB=&MakeVectorFromLinearColorRGB(Color=&ColorFromCurve(ColorCurve=DataInterface() {propertyValues {
        AlphaCurve {keys [<at 0 0 RCIM_Cubic>,<at 0.3821 1>,<at 1 0 RCIM_Cubic>]}
        BlueCurve {keys [
          <at 0 1 RCIM_Cubic>
          <at 0.4471 1>
          {arriveTangent=-1.3351, interpMode=RCIM_Cubic, leaveTangent=-1.3351, time=0.7702, value=0.2627}
          <at 0.9875 0.2785>
         ]}
        GreenCurve {keys [
          <at 0 1 RCIM_Cubic>
          <at 0.4471 1>
          {arriveTangent=-0.1735, interpMode=RCIM_Cubic, leaveTangent=-0.1735, time=0.7702, value=0.8637}
          <at 0.9875 0.9062>
         ]}
        RedCurve {keys [
          <at 0 1 RCIM_Cubic>
          <at 0.4471 1>
          {arriveTangent=-0.1866, interpMode=RCIM_Cubic, leaveTangent=-0.1866, time=0.7702, value=0.8698}
          <at 0.9875 0.8991>
         ]}
       }}))
    ScaleA=true
    ScaleRGB=true
   )
   /SetVariables() {isSetParameters=true}
   &UpdateMeshOrientation(
    Delta Time:NiagaraFloat=$Particles.Lifetime
    Rotation Coordinate Space=::NewEnumerator3
    Rotation Rate=&FloatFromCurve(FloatCurve=DataInterface() {propertyValues {Curve {keys [<at 0 0.17 RCIM_Cubic>, {arriveTangent=0.0289, interpMode=RCIM_Cubic, leaveTangent=0.0289, tangentMode=RCTM_User, time=1, value=0}]}}})
    Rotation Vector=<0 0 1>
   )
   &DynamicMaterialParameters(
    Index 0 Param 1=&FloatFromCurve(FloatCurve=DataInterface() {propertyValues {Curve {keys [
         {arriveTangent=1.9441, interpMode=RCIM_Cubic, leaveTangent=1.9441, tangentMode=RCTM_User, time=0, value=-1}
         {arriveTangent=1.1443, interpMode=RCIM_Cubic, leaveTangent=1.1443, tangentMode=RCTM_User, time=0.6, value=0.2}
         {arriveTangent=0.5487, interpMode=RCIM_Cubic, leaveTangent=0.5487, tangentMode=RCTM_User, time=1, value=0.7}
        ]}}})
    Index 0 Param 3=&FloatFromCurve(FloatCurve=DataInterface() {propertyValues {Curve {keys [
         {arriveTangent=-0.2355, interpMode=RCIM_Cubic, leaveTangent=-0.2355, tangentMode=RCTM_User, time=0, value=0}
         {arriveTangent=1.9444, interpMode=RCIM_Cubic, leaveTangent=1.9444, tangentMode=RCTM_User, time=1, value=0.7}
        ]}}})
    Index 0 Param 4=&FloatFromCurve(FloatCurve=DataInterface() {propertyValues {Curve {keys [
         {arriveTangent=-0.2255, interpMode=RCIM_Cubic, leaveTangent=-0.2255, tangentMode=RCTM_User, time=0, value=0}
         {arriveTangent=12.7998, interpMode=RCIM_Cubic, leaveTangent=12.7998, tangentMode=RCTM_User, time=0.2, value=1}
         {arriveTangent=-0.2255, interpMode=RCIM_Cubic, leaveTangent=-0.2255, tangentMode=RCTM_User, time=0.4, value=5}
         {arriveTangent=-0.0842, interpMode=RCIM_Cubic, leaveTangent=-0.0842, tangentMode=RCTM_User, time=1, value=5}
        ]}}})
    Index 1 Param 1=&FloatFromCurve(FloatCurve=DataInterface() {propertyValues {Curve {keys [<at 0 1 RCIM_Cubic>, {arriveTangent=-4.6529, interpMode=RCIM_Cubic, leaveTangent=-4.6529, tangentMode=RCTM_User, time=0.45, value=0}]}}})
    Module.Param0WriteEnabled=true
    Module.Param0WriteEnabled1=true
    Module.Param1WriteEnabled=true
    Module.Param1WriteEnabled1=true
    Module.Param2WriteEnabled=true
    Module.Param2WriteEnabled1=true
    Module.Param3WriteEnabled=true
    Module.Param3WriteEnabled1=true
    Write Parameter Index 1=true
   )
  ]
  rendererClasses [&NiagaraMeshRendererProperties]
  renderers [
   {rendererClass=&NiagaraMeshRendererProperties, rendererIndex=0}
   {rendererClass=&NiagaraMeshRendererProperties, rendererIndex=1}
  ]
 }
 Glow {
  emitterUpdateScript [
   &EmitterState(Life Cycle Mode=::NewEnumerator1, Loop Behavior=::NewEnumerator1, Loop Delay=0.06, Loop Duration=1, UseLoopDelay=true)
   &SpawnBurst_Instantaneous(Spawn Count=1, Use Spawn Probability=false)
  ]
  particleSpawnScript [
   &InitializeParticle(
    Color Mode=::NewEnumerator0
    Lifetime=0.26
    Mesh Scale Mode=::NewEnumerator3
    Mesh Uniform Scale=5.5
    UsePositionOffset=false
    Write Material Random=true
   )
   &SkeletalMeshLocation(
    Bone Sampling Mode=::NewEnumerator4
    MaskSpawnGroup=false
    Sampled Position Offset=<0 5 30>
    Sampled Position Offset Coordinate Space:ENiagaraSkelMeshTransforms=::NewEnumerator2
    Skeletal Mesh=DataInterface() {propertyValues {
      DefaultMesh=null
      ExcludeBoneName=root
      OptionalComponentClass=null
      PreviewMesh=&M_Med_Soldier_04_Preview
      SkinningMode=SkinOnTheFly
      SoftSourceActor=null
      SourceMode=Default
      UvSetIndex=0
      WholeMeshLOD=-1
      bExcludeBone=false
      bReadDeformedGeometry=true
      bRequireCurrentFrameData=true
      ComponentTags []
      FilteredBones [pelvis]
      FilteredSockets []
      MeshUserParameter {parameter {name=None, typeDefHandle {registeredTypeIndex=1}, varData []}}
      SamplingRegions []
     }}
   )
   &InitialMeshOrientation(
    Orientation Coordinate Space=::Local
    Rotation=<0.25 0 0>
    Rotation Coordinate Space=::NewEnumerator2
    Use Orientation Vector=true
    Use Rotation Vector=true
   )
   &Color(Color=<rgba 1 1 1 0.3>)
  ]
  particleUpdateScript [
   &ParticleState()
   &ScaleColor(
    Scale Alpha=&FloatFromCurve(FloatCurve=DataInterface() {propertyValues {Curve {keys [
         {arriveTangent=11.4923, interpMode=RCIM_Cubic, leaveTangent=11.4923, tangentMode=RCTM_User, time=0.1, value=0}
         {arriveTangent=0.6575, interpMode=RCIM_Cubic, leaveTangent=-0.4481, tangentMode=RCTM_Break, time=0.31, value=1}
         {arriveTangent=-2.0822, interpMode=RCIM_Cubic, leaveTangent=-1.2284, tangentMode=RCTM_Break, time=0.7, value=0.3}
         {interpMode=RCIM_Cubic, leaveTangent=-7.5208, time=1, value=0}
        ]}}})
    Scale RGB=&MakeVectorFromLinearColorRGB(Color=&ColorFromCurve(ColorCurve=DataInterface() {propertyValues {
        AlphaCurve {keys [<at 0 0 RCIM_Cubic>,<at 0.2086 1>,<at 1 0 RCIM_Cubic>]}
        BlueCurve {keys [
          <at 0 1 RCIM_Cubic>
          <at 0.4471 1>
          {arriveTangent=-0.9985, interpMode=RCIM_Cubic, leaveTangent=-0.9985, time=0.7702, value=0.005}
          <at 0.9875 0.014>
         ]}
        GreenCurve {keys [
          <at 0 1 RCIM_Cubic>
          <at 0.4471 1>
          {arriveTangent=-0.9019, interpMode=RCIM_Cubic, leaveTangent=-0.9019, time=0.7702, value=0.1599}
          <at 0.9875 0.1094>
         ]}
        RedCurve {keys [
          <at 0 1 RCIM_Cubic>
          <at 0.4471 1>
          {arriveTangent=-0.903, interpMode=RCIM_Cubic, leaveTangent=-0.903, time=0.7702, value=0.1615}
          <at 0.9875 0.1083>
         ]}
       }}))
    ScaleA=true
    ScaleRGB=true
   )
   &UpdateMeshOrientation(
    Delta Time:NiagaraFloat=$Particles.Lifetime
    Rotation Coordinate Space=::NewEnumerator3
    Rotation Rate=&FloatFromCurve(FloatCurve=DataInterface() {propertyValues {Curve {keys [
         {arriveTangent=-0.0721, interpMode=RCIM_Cubic, leaveTangent=-0.0721, tangentMode=RCTM_User, time=0, value=0.24}
         {arriveTangent=0.0791, interpMode=RCIM_Cubic, leaveTangent=0.0791, tangentMode=RCTM_User, time=0.7, value=0}
        ]}}})
    Rotation Vector=<0 0 1>
   )
   &DynamicMaterialParameters(
    Index 0 Param 1=&FloatFromCurve(FloatCurve=DataInterface() {propertyValues {Curve {keys [{arriveTangent=6.4629, interpMode=RCIM_Cubic, leaveTangent=6.4629, tangentMode=RCTM_User, time=0, value=-1}, <at 1 0.7 RCIM_Cubic>]}}})
    Index 0 Param 2=&FloatFromCurve(FloatCurve=DataInterface() {propertyValues {Curve {keys [<at 0 6>,<at 0.5 2>]}}})
    Index 0 Param 3=&FloatFromCurve(FloatCurve=DataInterface() {propertyValues {Curve {keys [
         {arriveTangent=0.0878, interpMode=RCIM_Cubic, leaveTangent=0.0878, tangentMode=RCTM_User, time=0, value=1}
         {arriveTangent=-1.1819, interpMode=RCIM_Cubic, leaveTangent=-1.1819, tangentMode=RCTM_User, time=1, value=0.45}
        ]}}})
    Index 0 Param 4=&FloatFromCurve(FloatCurve=DataInterface() {propertyValues {Curve {keys [<at 0 1>]}}})
    Index 1 Param 1=&FloatFromCurve(FloatCurve=DataInterface() {propertyValues {Curve {keys [<at 0 1 RCIM_Cubic>, {arriveTangent=-4.6529, interpMode=RCIM_Cubic, leaveTangent=-4.6529, tangentMode=RCTM_User, time=0.45, value=0}]}}})
    Module.Param0WriteEnabled=true
    Module.Param0WriteEnabled1=true
    Module.Param1WriteEnabled=true
    Module.Param1WriteEnabled1=true
    Module.Param2WriteEnabled=true
    Module.Param2WriteEnabled1=true
    Module.Param3WriteEnabled=true
    Module.Param3WriteEnabled1=true
    Write Parameter Index 1=true
   )
  ]
  rendererClasses [&NiagaraMeshRendererProperties]
  renderers [{rendererClass=&NiagaraMeshRendererProperties, rendererIndex=0}]
 }
 Light {
  emitterUpdateScript [
   &EmitterState(Life Cycle Mode=::NewEnumerator1, Loop Behavior=::NewEnumerator1, Loop Delay=0.06, Loop Duration=1, UseLoopDelay=true)
   &SpawnBurst_Instantaneous(Spawn Count=1, Use Spawn Probability=false)
  ]
  particleSpawnScript [
   &InitializeParticle(
    Color Mode=::NewEnumerator0
    Lifetime=0.31
    Mesh Scale=<5.6 5.6 3>
    Mesh Scale Mode=::NewEnumerator4
    UsePositionOffset=false
    Write Material Random=true
   )
   &SkeletalMeshLocation(
    Bone Sampling Mode=::NewEnumerator4
    MaskSpawnGroup=false
    Sampled Position Offset=<0 5 30>
    Sampled Position Offset Coordinate Space:ENiagaraSkelMeshTransforms=::NewEnumerator2
    Skeletal Mesh=DataInterface() {propertyValues {
      DefaultMesh=null
      ExcludeBoneName=root
      OptionalComponentClass=null
      PreviewMesh=&M_Med_Soldier_04_Preview
      SkinningMode=SkinOnTheFly
      SoftSourceActor=null
      SourceMode=Default
      UvSetIndex=0
      WholeMeshLOD=-1
      bExcludeBone=false
      bReadDeformedGeometry=true
      bRequireCurrentFrameData=true
      ComponentTags []
      FilteredBones [pelvis]
      FilteredSockets []
      MeshUserParameter {parameter {name=None, typeDefHandle {registeredTypeIndex=1}, varData []}}
      SamplingRegions []
     }}
   )
   &InitialMeshOrientation(
    Orientation Coordinate Space=::Local
    Rotation=<0.25 0 0>
    Rotation Coordinate Space=::NewEnumerator2
    Use Orientation Vector=true
    Use Rotation Vector=true
   )
  ]
  particleUpdateScript [
   &ParticleState()
   &ScaleColor(
    Scale Alpha=&FloatFromCurve(FloatCurve=DataInterface() {propertyValues {Curve {keys [
         {arriveTangent=14.969, interpMode=RCIM_Cubic, leaveTangent=14.969, tangentMode=RCTM_User, time=0, value=0}
         {arriveTangent=1.3168, interpMode=RCIM_Cubic, leaveTangent=-0.6685, tangentMode=RCTM_Break, time=0.3, value=3}
         {arriveTangent=-2.3753, interpMode=RCIM_Cubic, leaveTangent=-2.3753, tangentMode=RCTM_User, time=0.4, value=3}
         {arriveTangent=-9.4457, interpMode=RCIM_Cubic, leaveTangent=-9.4457, tangentMode=RCTM_User, time=0.9, value=0}
        ]}}})
    Scale RGB=&MakeVectorFromLinearColorRGB(Color=&ColorFromCurve(ColorCurve=DataInterface() {propertyValues {
        AlphaCurve {keys [<at 0 0 RCIM_Cubic>,<at 0.3141 1>,<at 1 0 RCIM_Cubic>]}
        BlueCurve {keys [
          <at 0 1 RCIM_Cubic>
          <at 0.4471 1>
          {arriveTangent=-0.9985, interpMode=RCIM_Cubic, leaveTangent=-0.9985, time=0.7702, value=0.005}
          <at 0.9875 0.014>
         ]}
        GreenCurve {keys [
          <at 0 1 RCIM_Cubic>
          <at 0.4471 1>
          {arriveTangent=-0.9019, interpMode=RCIM_Cubic, leaveTangent=-0.9019, time=0.7702, value=0.1599}
          <at 0.9875 0.1094>
         ]}
        RedCurve {keys [
          <at 0 1 RCIM_Cubic>
          <at 0.4471 1>
          {arriveTangent=-0.903, interpMode=RCIM_Cubic, leaveTangent=-0.903, time=0.7702, value=0.1615}
          <at 0.9875 0.1083>
         ]}
       }}))
    ScaleA=true
    ScaleRGB=true
   )
   &UpdateMeshOrientation(
    Delta Time:NiagaraFloat=$Particles.Lifetime
    Rotation Coordinate Space=::NewEnumerator3
    Rotation Rate=&FloatFromCurve(FloatCurve=DataInterface() {propertyValues {Curve {keys [
         {arriveTangent=-0.042, interpMode=RCIM_Cubic, leaveTangent=-0.042, tangentMode=RCTM_User, time=0, value=0.155}
         {arriveTangent=0.0444, interpMode=RCIM_Cubic, leaveTangent=0.0444, tangentMode=RCTM_User, time=0.85, value=0}
        ]}}})
    Rotation Vector=<0 0 1>
   )
   &DynamicMaterialParameters(
    Index 0 Param 1=&FloatFromCurve(FloatCurve=DataInterface() {propertyValues {Curve {keys [{arriveTangent=6.4629, interpMode=RCIM_Cubic, leaveTangent=6.4629, tangentMode=RCTM_User, time=0, value=-1}, <at 1 1 RCIM_Cubic>]}}})
    Index 0 Param 2=&FloatFromCurve(FloatCurve=DataInterface() {propertyValues {Curve {keys [
         {arriveTangent=-59.7155, interpMode=RCIM_Cubic, leaveTangent=-59.7155, tangentMode=RCTM_User, time=0, value=6}
         {arriveTangent=-8.5028, interpMode=RCIM_Cubic, leaveTangent=-8.5028, tangentMode=RCTM_User, time=0.1, value=2}
         <at 0.5 1 RCIM_Cubic>
        ]}}})
    Index 0 Param 3=&FloatFromCurve(FloatCurve=DataInterface() {propertyValues {Curve {keys [
         {arriveTangent=10.498, interpMode=RCIM_Cubic, leaveTangent=10.498, tangentMode=RCTM_User, time=0, value=0}
         {arriveTangent=0.5, interpMode=RCIM_Cubic, leaveTangent=0.5, time=0.2, value=1}
         {arriveTangent=-1.1154, interpMode=RCIM_Cubic, leaveTangent=-1.1154, tangentMode=RCTM_User, time=1, value=0.5}
        ]}}})
    Index 0 Param 4=&FloatFromCurve(FloatCurve=DataInterface() {propertyValues {Curve {keys [<at 0 0>,<at 0.15 0>,<at 0.55 0>]}}})
    Index 1 Param 1=&FloatFromCurve(FloatCurve=DataInterface() {propertyValues {Curve {keys [<at 0 1 RCIM_Cubic>, {arriveTangent=-4.6529, interpMode=RCIM_Cubic, leaveTangent=-4.6529, tangentMode=RCTM_User, time=0.45, value=0}]}}})
    Module.Param0WriteEnabled=true
    Module.Param0WriteEnabled1=true
    Module.Param1WriteEnabled=true
    Module.Param1WriteEnabled1=true
    Module.Param2WriteEnabled=true
    Module.Param2WriteEnabled1=true
    Module.Param3WriteEnabled=true
    Module.Param3WriteEnabled1=true
    Write Parameter Index 1=true
   )
  ]
  rendererClasses [&NiagaraMeshRendererProperties]
  renderers [{rendererClass=&NiagaraMeshRendererProperties, rendererIndex=0}]
 }
 Smoke {
  emitterUpdateScript [
   &EmitterState()
   &SpawnRate(SpawnRate=45, Use Spawn Probability=false)
  ]
  particleSpawnScript [
   &InitializeParticle(
    Color Mode=::NewEnumerator0
    Lifetime Max=0.77
    Lifetime Min=0.3
    Lifetime Mode=::NewEnumerator1
    Sprite Rotation Angle=&RandomRangeFloat(Minimum=0, Maximum=1)
    Sprite Rotation Mode=::NewEnumerator1
    Sprite Size Mode=::NewEnumerator1
    Uniform Sprite Size Max=100
    Uniform Sprite Size Min=40
    UsePositionOffset=false
    Write Material Random=true
   )
   &SkeletalMeshLocation(
    Bone Sampling Mode=::NewEnumerator4
    MaskSpawnGroup=false
    Skeletal Mesh=DataInterface() {propertyValues {
      DefaultMesh=null
      ExcludeBoneName=root
      OptionalComponentClass=null
      PreviewMesh=&M_Med_Soldier_04_Preview
      SkinningMode=SkinOnTheFly
      SoftSourceActor=null
      SourceMode=Default
      UvSetIndex=0
      WholeMeshLOD=-1
      bExcludeBone=false
      bReadDeformedGeometry=true
      bRequireCurrentFrameData=true
      ComponentTags []
      FilteredBones [root]
      FilteredSockets []
      MeshUserParameter {parameter {name=None, typeDefHandle {registeredTypeIndex=1}, varData []}}
      SamplingRegions []
     }}
   )
   &AddVelocityInCone(
    Cone Angle=40
    Cone Axis=&RotateVectorByQuaternion(Vector To Rotate=&RandomRangeVector(Minimum=<1 0 0>, Maximum=<1 0 0.3>), Quaternion=<0 0 0 1>)
    Module.Use Velocity Falloff On Cone Axis=false
    Velocity Strength=&FloatFromCurve(
     FloatCurve=DataInterface() {propertyValues {Curve {keys [<at 0 1 RCIM_Cubic>,<at 1 0.5>]}}}
     CurveIndex=$Emitter.NormalizedLoopAge
     Scale Curve=&RandomRangeFloat(Minimum=900, Maximum=1600)
    )
   )
   &SphereLocation(
    MaskSpawnGroup=false
    Override Seed=false
    Sphere Radius=5
    Surface Only=false
    UseNonUniformScale=false
    UseOffset=false
    UseSphereOrientation=false
   )
   &Color(Color=&RandomRangeLinearColor(Minimum=<rgba 0.23 0.2421 0.2552 0.18>, Maximum=<rgba 0.2476 0.2688 0.2917 0.15>))
  ]
  particleUpdateScript [
   &Drag(Drag=6, Use Linear Drag=true, Use Rotational Drag=false)
   &GravityForce(Gravity=&Lerp_Vector(A=<0 0 400>, B=<0 0 -100>, Alpha=$Particles.MaterialRandom))
   &SolveForcesAndVelocity(Clamp Velocity=false, Limit Acceleration=false)
   &ParticleState()
   &ScaleColor(
    Scale Alpha=&Multiply_Float(
     A=&FloatFromCurve(FloatCurve=DataInterface() {propertyValues {Curve {keys [<at 0 0>,<at 0.2049 0.9765>,<at 0.7791 0.9107>,<at 1 0>]}}})
     B=&Multiply_Float(A=&FNTimeOfDayScalar(Night Scale=1, Morning Scale=0.3, Evening Scale=0.3), B=1)
    )
    Scale RGB=&VectorFromFloat(Value=&FloatFromCurve(FloatCurve=DataInterface() {propertyValues {Curve {keys [<at 0 1>,<at 0.9983 0.5>]}}}))
    ScaleA=true
    ScaleRGB=true
   )
   &ScaleSpriteSize(Scale Factor=&Vector2DFromFloat(Value=&FloatFromCurve(FloatCurve=DataInterface() {propertyValues {Curve {keys [{arriveTangent=3.172, interpMode=RCIM_Cubic, time=0, value=0.15}, {interpMode=RCIM_Cubic, leaveTangent=1.6352, time=0.2, value=1}]}}})))
   &DynamicMaterialParameters(
    Index 0 Param 1=&FloatFromCurve(FloatCurve=DataInterface() {propertyValues {Curve {keys [
         {arriveTangent=0.9088, interpMode=RCIM_Cubic, leaveTangent=0.9088, tangentMode=RCTM_User, time=0, value=1}
         {arriveTangent=0.5473, interpMode=RCIM_Cubic, leaveTangent=0.5473, tangentMode=RCTM_User, time=1, value=2}
        ]}}})
    Index 0 Param 3=&FloatFromCurve(FloatCurve=DataInterface() {propertyValues {Curve {keys [
         {arriveTangent=-9.0447, interpMode=RCIM_Cubic, leaveTangent=-2.0044, tangentMode=RCTM_User, time=0, value=0.5}
         {interpMode=RCIM_Cubic, leaveTangent=4.6779, time=0.2851, value=0}
        ]}}})
    Index 1 Param 1=&FloatFromCurve(
     FloatCurve=DataInterface() {propertyValues {Curve {keys [<at 0 0>,<at 0.25 1>]}}}
     Scale Curve=&RandomRangeFloat(Minimum=2, Maximum=1)
    )
    Index 1 Param 2=&FloatFromCurve(
     FloatCurve=DataInterface() {propertyValues {Curve {keys [
         {interpMode=RCIM_Cubic, leaveTangent=3.7422, tangentMode=RCTM_Break, time=0, value=0.2}
         {interpMode=RCIM_Cubic, leaveTangent=0.9066, time=1, value=2}
        ]}}}
     Scale Curve=&RandomRangeFloat(Minimum=0.5, Maximum=1)
    )
    Index 1 Param 3=&FloatFromCurve(FloatCurve=DataInterface() {propertyValues {Curve {keys [
         {interpMode=RCIM_Cubic, leaveTangent=2.8445, tangentMode=RCTM_Break, time=0, value=0.2}
         {interpMode=RCIM_Cubic, leaveTangent=63.2978, time=0.25, value=1}
        ]}}})
    Index 1 Param 4=&FloatFromCurve(
     FloatCurve=DataInterface() {propertyValues {Curve {keys [<at 0 1>,<at 1 0.8>]}}}
     Scale Curve=&RandomRangeFloat(Minimum=0.5, Maximum=1)
    )
    Index 2 Param 1=&FloatFromCurve(FloatCurve=DataInterface() {propertyValues {Curve {keys [{interpMode=RCIM_Cubic, leaveTangent=3.2309, tangentMode=RCTM_Break, time=0.1, value=0}, <at 1 1>]}}})
    Index 2 Param 2=&FloatFromCurve(
     FloatCurve=DataInterface() {propertyValues {Curve {keys [<at 0 0.2 RCIM_Cubic>,<at 1 0.8 RCIM_Cubic>]}}}
     Scale Curve=&RandomRangeFloat(Minimum=0.2, Maximum=1.2)
    )
    Index 2 Param 3=&FloatFromCurve(FloatCurve=DataInterface() {propertyValues {Curve {keys [
         {interpMode=RCIM_Cubic, tangentMode=RCTM_Break, time=0, value=1}
         {interpMode=RCIM_Cubic, leaveTangent=11.1233, tangentMode=RCTM_Break, time=1, value=2}
        ]}}})
    Index 2 Param 4=&FloatFromCurve(
     FloatCurve=DataInterface() {propertyValues {Curve {keys [<at 0 1 RCIM_Cubic>, {interpMode=RCIM_Cubic, leaveTangent=-1.2956, tangentMode=RCTM_Break, time=1, value=0.75}]}}}
     Scale Curve=0.8
    )
    Module.Param0WriteEnabled=true
    Module.Param0WriteEnabled1=true
    Module.Param0WriteEnabled2=true
    Module.Param1WriteEnabled=true
    Module.Param1WriteEnabled1=true
    Module.Param1WriteEnabled2=true
    Module.Param2WriteEnabled=true
    Module.Param2WriteEnabled1=true
    Module.Param2WriteEnabled2=true
    Module.Param3WriteEnabled=true
    Module.Param3WriteEnabled1=true
    Module.Param3WriteEnabled2=true
    Write Parameter Index 1=true
    Write Parameter Index 2=true
   )
  ]
  rendererClasses [&NiagaraSpriteRendererProperties]
  renderers [{rendererClass=&NiagaraSpriteRendererProperties}]
 }
 Smoke001 {
  emitterUpdateScript [
   &EmitterState()
   &SpawnRate(SpawnRate=40, Use Spawn Probability=false)
  ]
  particleSpawnScript [
   &InitializeParticle(
    Color Mode=::NewEnumerator0
    Lifetime Max=0.7
    Lifetime Min=0.3
    Lifetime Mode=::NewEnumerator1
    Sprite Rotation Angle=&RandomRangeFloat(Minimum=0, Maximum=1)
    Sprite Rotation Mode=::NewEnumerator1
    Sprite Size Mode=::NewEnumerator1
    Uniform Sprite Size Max=100
    Uniform Sprite Size Min=40
    UsePositionOffset=false
    Write Material Random=true
   )
   &SkeletalMeshLocation(
    Bone Sampling Mode=::NewEnumerator4
    MaskSpawnGroup=false
    Skeletal Mesh=DataInterface() {propertyValues {
      DefaultMesh=null
      ExcludeBoneName=root
      OptionalComponentClass=null
      PreviewMesh=&M_Med_Soldier_04_Preview
      SkinningMode=SkinOnTheFly
      SoftSourceActor=null
      SourceMode=Default
      UvSetIndex=0
      WholeMeshLOD=-1
      bExcludeBone=false
      bReadDeformedGeometry=true
      bRequireCurrentFrameData=true
      ComponentTags []
      FilteredBones [root]
      FilteredSockets []
      MeshUserParameter {parameter {name=None, typeDefHandle {registeredTypeIndex=1}, varData []}}
      SamplingRegions []
     }}
   )
   &AddVelocityInCone(
    Cone Angle=20
    Cone Axis=&RotateVectorByQuaternion(Vector To Rotate=&RandomRangeVector(Minimum=<1 0 0.35>, Maximum=<1 0 1>), Quaternion=<0 0 0 1>)
    Module.Use Velocity Falloff On Cone Axis=false
    Velocity Strength=&FloatFromCurve(
     FloatCurve=DataInterface() {propertyValues {Curve {keys [<at 0 1 RCIM_Cubic>,<at 1 0.5>]}}}
     CurveIndex=$Emitter.NormalizedLoopAge
     Scale Curve=&RandomRangeFloat(Minimum=800, Maximum=1600)
    )
   )
   &SphereLocation(
    MaskSpawnGroup=false
    Override Seed=false
    Sphere Radius=5
    Surface Only=false
    UseNonUniformScale=false
    UseOffset=false
    UseSphereOrientation=false
   )
   &Color(Color=&RandomRangeLinearColor(Minimum=<rgba 0.23 0.2421 0.2552 0.18>, Maximum=<rgba 0.2476 0.2688 0.2917 0.15>))
  ]
  particleUpdateScript [
   &Drag(Drag=6, Use Linear Drag=true, Use Rotational Drag=false)
   &GravityForce(Gravity=&Lerp_Vector(A=<0 0 800>, B=<0 0 -20>, Alpha=$Particles.MaterialRandom))
   &SolveForcesAndVelocity(Clamp Velocity=false, Limit Acceleration=false)
   &ParticleState()
   &ScaleColor(
    Scale Alpha=&Multiply_Float(
     A=&FloatFromCurve(FloatCurve=DataInterface() {propertyValues {Curve {keys [<at 0 0>,<at 0.2049 0.9765>,<at 0.7791 0.9107>,<at 1 0>]}}})
     B=&Multiply_Float(A=&FNTimeOfDayScalar(Night Scale=1, Morning Scale=0.3, Evening Scale=0.3), B=1)
    )
    Scale RGB=&VectorFromFloat(Value=&FloatFromCurve(FloatCurve=DataInterface() {propertyValues {Curve {keys [<at 0 1>,<at 0.9983 0.5>]}}}))
    ScaleA=true
    ScaleRGB=true
   )
   &ScaleSpriteSize(Scale Factor=&Vector2DFromFloat(Value=&FloatFromCurve(FloatCurve=DataInterface() {propertyValues {Curve {keys [{arriveTangent=3.172, interpMode=RCIM_Cubic, time=0, value=0.15}, {interpMode=RCIM_Cubic, leaveTangent=1.6352, time=0.2, value=1}]}}})))
   &DynamicMaterialParameters(
    Index 0 Param 1=&FloatFromCurve(FloatCurve=DataInterface() {propertyValues {Curve {keys [
         {arriveTangent=0.9088, interpMode=RCIM_Cubic, leaveTangent=0.9088, tangentMode=RCTM_User, time=0, value=1}
         {arriveTangent=0.5473, interpMode=RCIM_Cubic, leaveTangent=0.5473, tangentMode=RCTM_User, time=1, value=2}
        ]}}})
    Index 0 Param 3=&FloatFromCurve(FloatCurve=DataInterface() {propertyValues {Curve {keys [
         {arriveTangent=-9.0447, interpMode=RCIM_Cubic, leaveTangent=-2.0044, tangentMode=RCTM_User, time=0, value=0.5}
         {interpMode=RCIM_Cubic, leaveTangent=4.6779, time=0.2851, value=0}
        ]}}})
    Index 1 Param 1=&FloatFromCurve(
     FloatCurve=DataInterface() {propertyValues {Curve {keys [<at 0 0>,<at 0.25 1>]}}}
     Scale Curve=&RandomRangeFloat(Minimum=2, Maximum=1)
    )
    Index 1 Param 2=&FloatFromCurve(
     FloatCurve=DataInterface() {propertyValues {Curve {keys [
         {interpMode=RCIM_Cubic, leaveTangent=3.7422, tangentMode=RCTM_Break, time=0, value=0.2}
         {interpMode=RCIM_Cubic, leaveTangent=0.9066, time=1, value=2}
        ]}}}
     Scale Curve=&RandomRangeFloat(Minimum=0.5, Maximum=1)
    )
    Index 1 Param 3=&FloatFromCurve(FloatCurve=DataInterface() {propertyValues {Curve {keys [
         {interpMode=RCIM_Cubic, leaveTangent=2.8445, tangentMode=RCTM_Break, time=0, value=0.2}
         {interpMode=RCIM_Cubic, leaveTangent=63.2978, time=0.25, value=1}
        ]}}})
    Index 1 Param 4=&FloatFromCurve(
     FloatCurve=DataInterface() {propertyValues {Curve {keys [<at 0 1>,<at 1 0.8>]}}}
     Scale Curve=&RandomRangeFloat(Minimum=0.5, Maximum=1)
    )
    Index 2 Param 1=&FloatFromCurve(FloatCurve=DataInterface() {propertyValues {Curve {keys [{interpMode=RCIM_Cubic, leaveTangent=3.2309, tangentMode=RCTM_Break, time=0.1, value=0}, <at 1 1>]}}})
    Index 2 Param 2=&FloatFromCurve(
     FloatCurve=DataInterface() {propertyValues {Curve {keys [<at 0 0.2 RCIM_Cubic>,<at 1 0.8 RCIM_Cubic>]}}}
     Scale Curve=&RandomRangeFloat(Minimum=0.2, Maximum=1.2)
    )
    Index 2 Param 3=&FloatFromCurve(FloatCurve=DataInterface() {propertyValues {Curve {keys [
         {interpMode=RCIM_Cubic, tangentMode=RCTM_Break, time=0, value=1}
         {interpMode=RCIM_Cubic, leaveTangent=11.1233, tangentMode=RCTM_Break, time=1, value=2}
        ]}}})
    Index 2 Param 4=&FloatFromCurve(
     FloatCurve=DataInterface() {propertyValues {Curve {keys [<at 0 1 RCIM_Cubic>, {interpMode=RCIM_Cubic, leaveTangent=-1.2956, tangentMode=RCTM_Break, time=1, value=0.75}]}}}
     Scale Curve=0.8
    )
    Module.Param0WriteEnabled=true
    Module.Param0WriteEnabled1=true
    Module.Param0WriteEnabled2=true
    Module.Param1WriteEnabled=true
    Module.Param1WriteEnabled1=true
    Module.Param1WriteEnabled2=true
    Module.Param2WriteEnabled=true
    Module.Param2WriteEnabled1=true
    Module.Param2WriteEnabled2=true
    Module.Param3WriteEnabled=true
    Module.Param3WriteEnabled1=true
    Module.Param3WriteEnabled2=true
    Write Parameter Index 1=true
    Write Parameter Index 2=true
   )
  ]
  rendererClasses [&NiagaraSpriteRendererProperties]
  renderers [{rendererClass=&NiagaraSpriteRendererProperties}]
 }
}
summary {
 systemName=NS_Emote_Quarrel_Male_Trail
 emitters [
  {emitterName=Dark, rendererClasses [&NiagaraMeshRendererProperties]}
  {emitterName=Smoke, rendererClasses [&NiagaraSpriteRendererProperties]}
  {emitterName=Light, rendererClasses [&NiagaraMeshRendererProperties]}
  {emitterName=Glow, rendererClasses [&NiagaraMeshRendererProperties]}
  {emitterName=Smoke001, rendererClasses [&NiagaraSpriteRendererProperties]}
 ]
 userVariables [
  User.RibbonIDOffset {type=NiagaraInt32}
  User.SkeletalMesh {
   type=NiagaraDataInterfaceSkeletalMesh
   defaultValue {
    struct=NiagaraExt_VariableValue_DataInterface
    value {dataInterface=&self:NiagaraDataInterfaceSkeletalMesh_2}
   }
  }
 ]
}
